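# Animation Blueprint for Smackie, an amphibian AI creature that hops, jumps, falls and swims.
# A Locomotion state machine runs on inherited flags, with additive leans on idle, hop and swim and turn-in-place shuffles in idle.
# An abducted idle overrides locomotion in tornadoes, then the pose passes through montage slots, quadruped slope warping and inertialization.
# Idle uses an overridable linked Fauna Idle layer through a cached pose, defaulting to a short idle.
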AnimBlueprint /Irwin/AI/Simple/Smackie/Animations/Smackie_AnimBP
target_skeleton=/Irwin/AI/Simple/Smackie/Meshes/Smackie_Amphibian_Skeleton
graphs [
 AnimGraph {nodes [
   AnimGraphNode_StateMachine_0 = Locomotion() {is=title AnimGraphNode_StateMachine}
   n1 = Blend Poses by bool(
    BlendPose_0=Sequence Player 'Smackie_AbductedIdle'() {sequence=&/Smackie_AbductedIdle}
    BlendPose_1=@AnimGraphNode_StateMachine_0
    BlendTime_0=0.5
    BlendTime_1=0.5
    bActiveValue=Get bIsInTornado() -> (bIsInTornado:bool)
   ) {blendType=Linear, transitionType=Inertialization}
   n2 = Quadruped Slope Warping!(ComponentPose:Component Space Pose Link Structure=Local To Component(LocalPose=Slot 'DefaultSlot'(Source=@n1)), Alpha:float=1) -> (Pose:Component Space Pose Link Structure) {
    is=title FortAnimGraphNode_QuadrupedSlopeWarping
    bOnlyAffectSkeletonRootBone=true
    maxStepHeight=35
    numSpineBones=6
    alphaBoolBlend {blendOption=Linear}
    armsFeetDefinitions [
     {numBonesInLimb=3, fKFootBone {boneName=BipedArm_A_Wrist_L}, iKFootBone {boneName=IK_Foot_Front_L}}
     {numBonesInLimb=3, fKFootBone {boneName=BipedArm_A_Wrist_R}, iKFootBone {boneName=IK_Foot_Front_R}}
    ]
    armsIKFootRootBone {boneName=IK_Foot_Front_Root}
    armsRootBone {boneName=QuadSpine_A_Chest_C}
    feetDefinitions [
     {numBonesInLimb=3, fKFootBone {boneName=BipedLeg_A_Ball_L}, iKFootBone {boneName=IK_Foot_Rear_L}}
     {numBonesInLimb=3, fKFootBone {boneName=BipedLeg_A_Ball_R}, iKFootBone {boneName=IK_Foot_Rear_R}}
    ]
    iKFootRootBone {boneName=IK_Foot_Rear_Root}
    pelvisBone {boneName=QuadSpine_A_Pelvis_C}
    skeletonRootBone {boneName=root}
   }
   n3 = Component To Local(ComponentPose=@n2)
   Output Pose(Result=Inertialization!(Source:Pose Link Structure=Slot 'SequencerDefault'(Source=@n3) {slotName=SequencerDefault}) -> (Pose:Pose Link Structure) {is=title AnimGraphNode_Inertialization}) {layerGroup=None}
   Save cached pose 'FaunaIdlePose'(Pose=Fauna_Idle_AnimInterface - Fauna Idle!(InPose Idle:Pose Link Structure) -> (Pose:Pose Link Structure) {is=title AnimGraphNode_LinkedAnimLayer, interface=/Irwin/AI/Interfaces/Fauna_Idle_AnimInterface.Fauna_Idle_AnimInterface_C, layer=Fauna Idle}) {cacheName=FaunaIdlePose}
  ]}
 Locomotion {
  entry=Idle
  path=AnimGraph.AnimGraphNode_StateMachine_0.Locomotion
  ids=id
  nodes [
   AnimStateNode_0 = Idle() {
    stateEntered {guid=#1}
    stateFullyBlended {guid=#2}
    stateLeft {guid=#3}
   }
   AnimStateNode_1 = Hop() {is=title AnimStateNode}
   AnimStateNode_2 = Falling() {
    is=title AnimStateNode
    stateEntered {guid=#4}
    stateFullyBlended {guid=#5}
    stateLeft {guid=#6}
   }
   AnimStateNode_3 = Swimming() {is=title AnimStateNode}
   AnimStateNode_4 = Jump() {
    is=title AnimStateNode
    stateEntered {guid=#7}
    stateFullyBlended {guid=#8}
    stateLeft {guid=#9}
   }
   AnimStateTransitionNode_0 = AnimStateTransitionNode() {logicType=TLT_Inertialization}
   AnimStateTransitionNode_1 = AnimStateTransitionNode() {crossfadeDuration=0.4, logicType=TLT_Inertialization, priorityOrder=2}
   AnimStateTransitionNode_10 = AnimStateTransitionNode() {crossfadeDuration=0.1, logicType=TLT_Inertialization}
   AnimStateTransitionNode_11 = AnimStateTransitionNode() {crossfadeDuration=0, logicType=TLT_Inertialization}
   AnimStateTransitionNode_12 = AnimStateTransitionNode() {crossfadeDuration=0.1, logicType=TLT_Inertialization, priorityOrder=2}
   AnimStateTransitionNode_14 = AnimStateTransitionNode() {bAutomaticRuleBasedOnSequencePlayerInState=true, crossfadeDuration=0.5, logicType=TLT_Inertialization}
   AnimStateTransitionNode_3 = AnimStateTransitionNode() {crossfadeDuration=0.3, logicType=TLT_Inertialization, priorityOrder=3}
   AnimStateTransitionNode_4 = AnimStateTransitionNode() {logicType=TLT_Inertialization}
   AnimStateTransitionNode_5 = AnimStateTransitionNode() {crossfadeDuration=0.4, logicType=TLT_Inertialization, priorityOrder=100}
   AnimStateTransitionNode_6 = AnimStateTransitionNode() {crossfadeDuration=0.1, logicType=TLT_Inertialization, priorityOrder=100}
   AnimStateTransitionNode_7 = AnimStateTransitionNode() {crossfadeDuration=0.4, logicType=TLT_Inertialization, priorityOrder=100}
   AnimStateTransitionNode_8 = AnimStateTransitionNode() {crossfadeDuration=0.3, logicType=TLT_Inertialization, priorityOrder=100}
   AnimStateTransitionNode_9 = AnimStateTransitionNode() {crossfadeDuration=0.1, logicType=TLT_Inertialization, priorityOrder=2}
  ]
  links [
   {
    dst=Hop
    rule=AnimGraph.AnimGraphNode_StateMachine_0.Locomotion.AnimStateTransitionNode_11.Transition
    src=Falling
    via=AnimStateTransitionNode_11
   }
   {
    dst=Idle
    rule=AnimGraph.AnimGraphNode_StateMachine_0.Locomotion.AnimStateTransitionNode_9.Transition
    src=Falling
    via=AnimStateTransitionNode_9
   }
   {
    dst=Swimming
    rule=AnimGraph.AnimGraphNode_StateMachine_0.Locomotion.AnimStateTransitionNode_7.Transition
    src=Falling
    via=AnimStateTransitionNode_7
   }
   {
    dst=Falling
    rule=AnimGraph.AnimGraphNode_StateMachine_0.Locomotion.AnimStateTransitionNode_3.Transition
    src=Hop
    via=AnimStateTransitionNode_3
   }
   {
    dst=Idle
    rule=AnimGraph.AnimGraphNode_StateMachine_0.Locomotion.AnimStateTransitionNode_4.Transition
    src=Hop
    via=AnimStateTransitionNode_4
   }
   {
    dst=Jump
    rule=AnimGraph.AnimGraphNode_StateMachine_0.Locomotion.AnimStateTransitionNode_12.Transition
    src=Hop
    via=AnimStateTransitionNode_12
   }
   {
    dst=Swimming
    rule=AnimGraph.AnimGraphNode_StateMachine_0.Locomotion.AnimStateTransitionNode_5.Transition
    src=Hop
    via=AnimStateTransitionNode_5
   }
   {
    dst=Falling
    rule=AnimGraph.AnimGraphNode_StateMachine_0.Locomotion.AnimStateTransitionNode_1.Transition
    src=Idle
    via=AnimStateTransitionNode_1
   }
   {
    dst=Hop
    rule=AnimGraph.AnimGraphNode_StateMachine_0.Locomotion.AnimStateTransitionNode_0.Transition
    src=Idle
    via=AnimStateTransitionNode_0
   }
   {
    dst=Jump
    rule=AnimGraph.AnimGraphNode_StateMachine_0.Locomotion.AnimStateTransitionNode_6.Transition
    src=Idle
    via=AnimStateTransitionNode_6
   }
   {
    dst=Falling
    rule=AnimGraph.AnimGraphNode_StateMachine_0.Locomotion.AnimStateTransitionNode_14.Transition
    src=Jump
    via=AnimStateTransitionNode_14
   }
   {
    dst=Hop
    rule=AnimGraph.AnimGraphNode_StateMachine_0.Locomotion.AnimStateTransitionNode_8.Transition
    src=Swimming
    via=AnimStateTransitionNode_8
   }
   {
    dst=Jump
    rule=AnimGraph.AnimGraphNode_StateMachine_0.Locomotion.AnimStateTransitionNode_10.Transition
    src=Swimming
    via=AnimStateTransitionNode_10
   }
  ]
 }
 Idle {
  path=AnimGraph.AnimGraphNode_StateMachine_0.Locomotion.AnimStateNode_0.Idle
  nodes [
   n1 = Get IdleShufflePlayRate() -> (IdleShufflePlayRate:float)
   n2 = Blend Poses by bool(
    BlendPose_0=Sequence Player 'Smackie_TurnInPlace_R_90_HipsAndFeet'!(PlayRate:float=@n1) -> (Pose) {sequence=&/Smackie_TurnInPlace_R_90_HipsAndFeet}
    BlendPose_1=Sequence Player 'Smackie_TurnInPlace_L_90_HipsAndFeet'!(PlayRate:float=@n1) -> (Pose) {sequence=&/Smackie_TurnInPlace_L_90_HipsAndFeet}
    BlendTime_0=0.4
    BlendTime_1=0.4
    bActiveValue=Get bIdleShuffleTurnDirection() -> (bIdleShuffleTurnDirection:bool)
   ) {blendType=Linear, transitionType=Inertialization}
   n3 = Use cached pose 'FaunaIdlePose'()
   n4 = Get bIsIdleShuffleTurn() -> (bIsIdleShuffleTurn:bool)
   n5 = Get IdleShuffleBodyBend() -> (IdleShuffleBodyBend:float)
   Output Animation Pose(Result=Apply Additive(
     Base=Blend Poses by bool(BlendPose_0=@n2, BlendPose_1=@n3, BlendTime_0=0.4, BlendTime_1=0.4, bActiveValue=@n4) {blendType=Linear, transitionType=Inertialization}
     Additive=Blendspace Player 'Smackie_Leans_BS'!(X=@n5) -> (Pose) {blendSpace=&/Smackie_Leans_BS}
    ) {alphaBoolBlend {blendOption=Linear}})
  ]
 }
 Hop {
  path=AnimGraph.AnimGraphNode_StateMachine_0.Locomotion.AnimStateNode_1.Hop
  nodes [
   n1 = Get IdleShuffleBodyBend()
   Output Animation Pose(Result=Apply Additive(
     Base=Sequence Player 'Smackie_Hop'() {sequence=&/Smackie_Hop}
     Additive=Blendspace Player 'Smackie_Leans_BS'(X=@n1) {blendSpace=&Smackie_Leans_BS}
    ) {alphaBoolBlend {blendOption=Linear}})
  ]
 }
 Falling {
  path=AnimGraph.AnimGraphNode_StateMachine_0.Locomotion.AnimStateNode_2.Falling
  nodes [Output Animation Pose(Result=Sequence Player 'Smackie_FallLoop'() {sequence=&/Smackie_FallLoop})]
 }
 Swimming {
  path=AnimGraph.AnimGraphNode_StateMachine_0.Locomotion.AnimStateNode_3.Swimming
  nodes [
   n1 = Sequence Player 'Smackie_Swimming'() {playRate=1.4, sequence=&/Smackie_Swimming}
   n2 = Sequence Player 'Smackie_SwimIdle'() {playRate=1.3, sequence=&/Smackie_SwimIdle}
   n3 = Get bCanTransitionIdleToHop() -> (bCanTransitionIdleToHop:bool)
   n4 = Get IdleShuffleBodyBend()
   Output Animation Pose(Result=Apply Additive(
     Base=Blend Poses by bool(BlendPose_0=@n1, BlendPose_1=@n2, BlendTime_0=0.4, BlendTime_1=0.6, bActiveValue=@n3) {blendType=Linear, transitionType=Inertialization}
     Additive=Blendspace Player 'Smackie_Leans_BS'(X=@n4) {blendSpace=&Smackie_Leans_BS}
    ) {alphaBoolBlend {blendOption=Linear}})
  ]
 }
 Jump {
  path=AnimGraph.AnimGraphNode_StateMachine_0.Locomotion.AnimStateNode_4.Jump
  nodes [Output Animation Pose(Result=Sequence Player 'Smackie_JumpStart'() {playRate=0.5, sequence=&/Smackie_JumpStart})]
 }
 Transition {
  path=AnimGraph.AnimGraphNode_StateMachine_0.Locomotion.AnimStateTransitionNode_0.Transition
  nodes [Result(bCanEnterTransition=Get bCanTransitionIdleToHop())]
 }
 Transition {
  path=AnimGraph.AnimGraphNode_StateMachine_0.Locomotion.AnimStateTransitionNode_1.Transition
  nodes [Result(bCanEnterTransition=Get bIsFalling())]
 }
 Transition {
  path=AnimGraph.AnimGraphNode_StateMachine_0.Locomotion.AnimStateTransitionNode_10.Transition
  nodes [Result(bCanEnterTransition=Get bIsJumping())]
 }
 Transition {
  path=AnimGraph.AnimGraphNode_StateMachine_0.Locomotion.AnimStateTransitionNode_11.Transition
  nodes [Result(bCanEnterTransition=Get bCanTransitionFallToHop() -> (bCanTransitionFallToHop:bool))]
 }
 Transition {
  path=AnimGraph.AnimGraphNode_StateMachine_0.Locomotion.AnimStateTransitionNode_12.Transition
  nodes [Result(bCanEnterTransition=Get bIsJumping())]
 }
 Transition {path=AnimGraph.AnimGraphNode_StateMachine_0.Locomotion.AnimStateTransitionNode_14.Transition, nodes [Result()]}
 Transition {
  path=AnimGraph.AnimGraphNode_StateMachine_0.Locomotion.AnimStateTransitionNode_3.Transition
  nodes [Result(bCanEnterTransition=Get bIsFalling())]
 }
 Transition {
  path=AnimGraph.AnimGraphNode_StateMachine_0.Locomotion.AnimStateTransitionNode_4.Transition
  nodes [Result(bCanEnterTransition=Get bCanTransitionHopToIdle() -> (bCanTransitionHopToIdle:bool))]
 }
 Transition {
  path=AnimGraph.AnimGraphNode_StateMachine_0.Locomotion.AnimStateTransitionNode_5.Transition
  nodes [Result(bCanEnterTransition=Get bIsSurfaceSwimming())]
 }
 Transition {
  path=AnimGraph.AnimGraphNode_StateMachine_0.Locomotion.AnimStateTransitionNode_6.Transition
  nodes [Result(bCanEnterTransition=Get bIsJumping())]
 }
 Transition {
  path=AnimGraph.AnimGraphNode_StateMachine_0.Locomotion.AnimStateTransitionNode_7.Transition
  nodes [Result(bCanEnterTransition=Get bIsSurfaceSwimming())]
 }
 Transition {
  path=AnimGraph.AnimGraphNode_StateMachine_0.Locomotion.AnimStateTransitionNode_8.Transition
  nodes [Result(bCanEnterTransition=NOT Boolean(A=Get bIsSurfaceSwimming()))]
 }
 Transition {
  path=AnimGraph.AnimGraphNode_StateMachine_0.Locomotion.AnimStateTransitionNode_9.Transition
  nodes [Result(bCanEnterTransition=Get bCanTransitionFallToIdle() -> (bCanTransitionFallToIdle:bool))]
 }
 EventGraph {nodes [Event BlueprintUpdateAnimation()]}
 Fauna Idle {nodes [
   Input Pose() {name=InPose Idle}
   Output Pose(Result=Sequence Player 'Smackie_Idle_Short'() {sequence=&/Smackie_Idle_Short}) {layerGroup=FaunaIdleOverride}
  ]}
 Fauna IdleTurn {nodes [Output Pose(Result=Input Pose() {name=InPose IdleTurn}) {layerGroup=FaunaIdleOverride}]}
]
variables []
